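# Translucent, unlit, two-sided Niagara mesh-particle material for burst spike effects on the BoomShot glider.
# Panned wispy noise subtracted from a UV V gradient drives both a dark-to-base color blend and a smoothstepped opacity shape.
# Particle color tints emissive and scales alpha, while an end fade subtracts opacity along V.
# Dynamic parameters control vertex offsets that stretch along the tangent-normal cross product and spread along the normal, weighted by V.
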Material /BRCosmetics/Gadgets/Assets/VinderTech_GliderChute/Glider_BoomShot/FX/M_Glider_BoomShot_BurstSpike
graph {
 properties {
  bAutomaticallySetUsageInEditor=false
  bUsedWithNiagaraMeshParticles=true
  blendMode=BLEND_Translucent
  shadingModel=MSM_Unlit
  twoSided=true
  usedShadingModels=Unlit
 }
 nodes [
  n1 = VertexNormalWS()
  n2 = CrossProduct(A=VertexTangentWS(), B=@n1)
  n3 = OneMinus(None=ComponentMask(None=TextureCoordinate()) {g=true})
  n4 = DynamicParameter() {defaultValue=<rgba 0 0 1 1>, paramNames [WPO Stretch, WPO Spread, Param3, Param4]}
  n5 = Subtract(A=@n3) {constB=0.5}
  Add(A=Multiply(A=Multiply(A=@n2, B=@n3), B=@n4.WPO Stretch), B=Multiply(A=Multiply(A=@n1, B=@n5), B=@n4.WPO Spread))
  n6 = VectorParameter() {group=Shading, parameterName=Dark Color, sortPriority=1, defaultValue=<rgba 0.3698 0 0.2705 1>}
  n7 = VectorParameter() {group=Shading, parameterName=Base Color, sortPriority=0, defaultValue=<rgba 2.5 0.1725 1.8913 1>}
  n8 = TextureCoordinate()
  n9 = Multiply(
   A=TextureCoordinate()
   B=AppendVector(
    A=ScalarParameter() {defaultValue=3, parameterName=Tile X, sliderMax=5}
    B=ScalarParameter() {defaultValue=0.2, parameterName=Tile Y, sliderMax=5}
   )
  )
  n10 = AppendVector(A=ScalarParameter() {parameterName=Speed X, sliderMax=5}, B=ScalarParameter() {defaultValue=0.5, parameterName=Speed Y, sliderMax=5})
  n11 = Subtract(
   A=Multiply(A=ComponentMask(None=@n8) {g=true}) {constB=2}
   B=TextureSample(UVs=Panner(Coordinate=@n9, Speed=@n10)) {samplerType=SAMPLERTYPE_Grayscale, texture=/Game/Effects/Fort_Effects/Textures/Fire/T_WispyFireNoise_Soft_M}.R
  )
  n12 = SmoothStep(
   Min=ScalarParameter() {parameterName=Color Min, sliderMax=5, sliderMin=-5}
   Max=ScalarParameter() {defaultValue=1, parameterName=Color Max, sliderMax=5, sliderMin=-5}
   Value=@n11
  ) {constMax=1.5, constMin=0.8}
  n13 = ParticleColor()
  MaterialFunctionCall(Emissive=Multiply(A=LinearInterpolate(A=@n6.RGB, B=@n7.RGB, Alpha=@n12), B=@n13.RGB)) {materialFunction=&MF_FX_EmissiveAdjustments}
  n14 = ScalarParameter() {defaultValue=0.5, parameterName=Shape Min, sliderMax=5, sliderMin=-5}
  n15 = ScalarParameter() {defaultValue=1, parameterName=Shape Max, sliderMax=5, sliderMin=-5}
  n16 = ScalarParameter() {defaultValue=0.75, parameterName=End Fade Start, sliderMax=1}
  n17 = ComponentMask(None=TextureCoordinate()) {g=true}
  Subtract(
   A=Multiply(A=@n13.A, B=SmoothStep(Min=@n14, Max=@n15, Value=@n11) {constMax=0.8, constMin=0.6})
   B=Multiply(A=SmoothStep(Min=@n16, Value=@n17) {constMin=0.75}, B=ScalarParameter() {defaultValue=1, parameterName=End Fade Intensity, sliderMax=10}) {constB=1.5}
  )
 ]
}
