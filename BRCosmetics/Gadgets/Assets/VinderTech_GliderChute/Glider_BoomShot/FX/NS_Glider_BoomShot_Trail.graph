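# Niagara trail effect for the BoomShot glider's twin thrusters: flames with glow and lights, spikes, embers, smoke, motion trails and speed lines.
# Emitters attach to two skeletal mesh bones or sockets; age curves stage an ignition around two seconds in, with system Timeline values boosting intensity.
# Smoke puff meshes follow Smoke_Ribbon through a particle reader, and the ribbon bends using smoothed InterpolateOverTime values.
# Front-end modifiers swap directions and sizes between gameplay and menu display.
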NiagaraSystem /BRCosmetics/Gadgets/Assets/VinderTech_GliderChute/Glider_BoomShot/FX/NS_Glider_BoomShot_Trail
dependencies {
 usedDataInterfaces [
  &NiagaraDataInterfaceArrayFloat
  &NiagaraDataInterfaceArrayInt32
  &NiagaraDataInterfaceCurve
  &NiagaraDataInterfaceColorCurve
  &NiagaraDataInterfaceVector2DCurve
  &NiagaraDataInterfaceParticleRead
  &NiagaraDataInterfaceArrayColor
  &NiagaraDataInterfaceSkeletalMesh
 ]
 usedDynamicInputs [
  &SetBoolByFloatComparison
  &RemapRange
  &MakeIntFromFloat
  &NormalizeVector
  &FNTimeOfDayColor
 ]
 usedModules [
  &self:SystemSpawnScript.SystemScriptSource.SystemScriptGraph.NiagaraNodeAssignment_0.SetVariables_#1
  &SystemState
  &Timeline
  &FN_Integrator
  &InterpolateOverTime
  &self:SystemSpawnScript.SystemScriptSource.SystemScriptGraph.NiagaraNodeAssignment_1.SetVariables_#2
 ]
}
emitters {
 Embers {
  emitterUpdateScript [
   &EmitterState(Life Cycle Mode=::NewEnumerator1, Loop Behavior=::NewEnumerator1, Loop Delay=2.2, Loop Duration Mode=::NewEnumerator1, UseLoopDelay=true)
   &SpawnBurst_Instantaneous(Spawn Count=200, Use Spawn Probability=false, UseLoopCountLimit=false)
   &SpawnRate(
    SpawnRate=&FloatFromCurve(
     FloatCurve=DataInterface() {propertyValues {Curve {keys [
         {arriveTangent=31.4523, interpMode=RCIM_Cubic, leaveTangent=31.4523, tangentMode=RCTM_User, time=2.2, value=0}
         {arriveTangent=0.0777, interpMode=RCIM_Cubic, leaveTangent=0.0777, tangentMode=RCTM_User, time=2.3, value=2}
         <at 3 1 RCIM_Cubic>
        ]}}}
     CurveIndex=$Emitter.Age
     Scale Curve=200
    )
    Use Spawn Probability=true
   )
   &SpawnRate(
    Spawn Probability=&FloatFromCurve(
     FloatCurve=DataInterface() {propertyValues {Curve {keys [{interpMode=RCIM_Cubic, tangentMode=RCTM_User, time=0, value=0}, {interpMode=RCIM_Cubic, tangentMode=RCTM_User, time=1, value=1}]}}}
     CurveIndex=$System.Timeline.Time
    )
    SpawnRate=200
    Use Spawn Probability=true
   ) {module=SpawnRate001}
  ]
  particleSpawnScript [
   /SetVariables(Particles.PerParticleRandom:NiagaraFloat=&RandomRangeFloat(Minimum=0, Maximum=1)) {isSetParameters=true, module=SetVariables_#3}
   &FN_SetSpawnGroup(Number of Spawn Groups=2)
   &InitializeParticle(
    Color=&LerpLinearColors(
     StartColor=<rgba 0.9 0.225 0.09 1>
     EndColor=<rgba 1 1 1 1>
     LerpFactor=&ClampFloat(Float=&Add_Float(
       A=&FloatFromCurve(
        FloatCurve=DataInterface() {propertyValues {Curve {keys [
            {arriveTangent=25.4307, interpMode=RCIM_Cubic, leaveTangent=25.4307, tangentMode=RCTM_User, time=2.1, value=0}
            {arriveTangent=-0.0018, interpMode=RCIM_Cubic, leaveTangent=-0.0018, tangentMode=RCTM_User, time=2.2, value=1}
            {arriveTangent=0.0249, interpMode=RCIM_Cubic, leaveTangent=0.0249, tangentMode=RCTM_User, time=3, value=0}
           ]}}}
        CurveIndex=$Emitter.Age
       )
       B=&FloatFromCurve(
        FloatCurve=DataInterface() {propertyValues {Curve {keys [{tangentMode=RCTM_User, time=0, value=0}, {tangentMode=RCTM_User, time=1, value=1}]}}}
        CurveIndex=$System.Timeline.Time
       )
      ))
    )
    Lifetime=&Multiply_Float(
     A=&Lerp_Float(A=0.15, B=0.3, Alpha=$Particles.PerParticleRandom)
     B=&Lerp_Float(A=0.8, B=0.75, Alpha=$System.Timeline.Time)
    )
    Mesh Scale=<2 1 1>
    Mesh Scale Mode=::NewEnumerator4
    Sprite Size=&Multiply_Vector2D(
     A=<1 1.5>
     B=&MakeVector2D(
      X=&FloatFromCurve(
       FloatCurve=DataInterface() {propertyValues {Curve {keys [<at 0 1>,<at 0.8 2 RCIM_Constant>,<at 0.95 3 RCIM_Constant>]}}}
       CurveIndex=$Particles.PerParticleRandom
      )
      Y=&FloatFromCurve(
       FloatCurve=DataInterface() {propertyValues {Curve {keys [<at 0 1>,<at 0.8 2 RCIM_Constant>,<at 0.95 3>]}}}
       CurveIndex=$Particles.PerParticleRandom
      )
     )
    )
    Sprite Size Mode=::NewEnumerator4
    UsePositionOffset=false
    Write Material Random=true
   )
   &SkeletalMeshLocation(
    Bone / Socket Index=&SpawnGroup()
    Interpolate Mesh Sampling=false
    MaskSpawnGroup=false
    Mesh Sampling Type=::NewEnumerator5
    Sampled Position Offset=&Multiply_VectorByFloat(Vector=&MaskVectorBySpawnGroup(Masked Vector=<0 1 0>, Passthrough Vector=<0 -1 0>), Float=20)
    Skeletal Mesh=$System.Skeletal Mesh
    Socket Sampling Mode=::NewEnumerator5
   )
   &ShapeLocation(
    Cylinder Height=&FloatFromCurve(
     FloatCurve=DataInterface() {propertyValues {Curve {keys [{interpMode=RCIM_Cubic, tangentMode=RCTM_User, time=0, value=1}, {interpMode=RCIM_Cubic, tangentMode=RCTM_User, time=1, value=1.75}]}}}
     CurveIndex=$System.Timeline.Time
     Scale Curve=30
    )
    Cylinder Height Midpoint=1
    Cylinder Radius=&FloatFromCurve(
     FloatCurve=DataInterface() {propertyValues {Curve {keys [{interpMode=RCIM_Cubic, tangentMode=RCTM_User, time=0, value=1}, {interpMode=RCIM_Cubic, tangentMode=RCTM_User, time=1, value=0.8}]}}}
     CurveIndex=$System.Timeline.Time
     Scale Curve=15
    )
    MaskSpawnGroup=false
    Override Seed=false
    Rotation Axis=&FN_FrontEndVector3Modifier(Gameplay=<0 1 0>, Front End=<0 -1 0>)
    Rotation Mode=::NewEnumerator1
    Shape Primitive=::NewEnumerator1
    Surface Only=false
    UseOwnerScale=true
   )
   &AddVelocity(
    Velocity=&Add_Vector(
     A=$Particles.ShapeLocation.ShapeNormal
     B=&FN_FrontEndVector3Modifier(Gameplay=<0 -3 0>, Front End=<0 3 0>)
    )
    Velocity Speed Scale=&Multiply_Float(
     A=&Lerp_Float(A=50, B=75, Alpha=$Particles.PerParticleRandom)
     B=&Multiply_Float(
      A=&Lerp_Float(A=1, B=2, Alpha=$System.Timeline.Time)
      B=&FloatFromCurve(FloatCurve=DataInterface() {propertyValues {Curve {keys [<at 2 2.5 RCIM_Cubic>,<at 2.5 1 RCIM_Cubic>]}}}, CurveIndex=$Emitter.Age)
     )
    )
   )
  ]
  particleUpdateScript [
   &ParticleState()
   &Drag(
    Drag=&FloatFromCurve(
     FloatCurve=DataInterface() {propertyValues {Curve {keys [
         {interpMode=RCIM_Cubic, tangentMode=RCTM_User, time=0, value=0.8}
         {arriveTangent=0.4778, interpMode=RCIM_Cubic, leaveTangent=0.4778, tangentMode=RCTM_User, time=1, value=1.0375}
        ]}}}
     Scale Curve=5
    )
    Use Linear Drag=true
    Use Rotational Drag=false
   )
   &CurlNoiseForce(Noise Frequency=100, Noise Strength=750, Pan Noise=true, Pan Noise Field=<0 0 1>) {module=CurlNoiseForce001}
   &GravityForce(
    Coordinate Space=::Simulation
    Gravity=&Multiply_VectorByFloat(
     Vector=<1 0 0>
     Float=&Multiply_Float(
      A=&Multiply_Float(
       A=&Lerp_Float(A=250, B=375, Alpha=$System.Timeline.Time)
       B=&Lerp_Float(A=1, B=1.5, Alpha=$Particles.PerParticleRandom)
      )
      B=$System.InterpolateOverTime001.InterpolateOverTime_Float.Moving Average
     )
    )
   )
   &SolveForcesAndVelocity(Clamp Velocity=false, Limit Acceleration=false)
   &ScaleColor(
    Scale RGB=&MakeVectorFromLinearColorRGB(Color=&ColorFromCurve(ColorCurve=DataInterface() {propertyValues {
        AlphaCurve {keys [<at 0 1 RCIM_Cubic>]}
        BlueCurve {keys [<at 0.0088 0>,<at 0.6353 0.5 RCIM_Cubic>]}
        GreenCurve {keys [<at 0.0088 1.75>,<at 0.6353 0 RCIM_Cubic>]}
        RedCurve {keys [<at 0.0088 3>,<at 0.6353 1 RCIM_Cubic>]}
       }}))
    ScaleA=true
    ScaleRGB=true
   )
   &ScaleSpriteSize(
    Uniform Curve Scale=&MakeFloatFromVector(Vector=$Engine.Owner.Scale, Channel=::NewEnumerator0)
    Uniform Curve Sprite Scale=DataInterface() {propertyValues {Curve {keys [
        {tangentMode=RCTM_User, time=0, value=1}
        {arriveTangent=-0.0405, interpMode=RCIM_Cubic, leaveTangent=-0.0405, tangentMode=RCTM_User, time=0.7009, value=1}
        {arriveTangent=-8.0089, interpMode=RCIM_Cubic, leaveTangent=-8.0089, tangentMode=RCTM_User, time=1, value=0}
       ]}}}
   )
   &ScaleSpriteSizeBySpeed(
    Max Scale Factor=<0.75 5>
    Min Scale Factor=<1.5 1>
    Sample Scale Factor By Curve=true
    Scale Factor Curve=DataInterface() {propertyValues {Curve {keys [<at 0 0>, {arriveTangent=1.4989, interpMode=RCIM_Cubic, tangentMode=RCTM_Break, time=1, value=1}]}}}
    Velocity Threshold=250
   )
  ]
  rendererClasses [&NiagaraSpriteRendererProperties]
  renderers [{rendererClass=&NiagaraSpriteRendererProperties}]
 }
 Motion_Trails {
  emitterSpawnScript [&FN_RibbonIDOffsetUtility()]
  emitterUpdateScript [
   &EmitterState(
    Inactive Response=::NewEnumerator1
    Life Cycle Mode=::NewEnumerator1
    Loop Behavior=::NewEnumerator1
    Loop Delay=0.4
    Loop Duration=1.5
    UseLoopDelay=true
   )
   &FN_SpawnAnimTrail(Number of Anim Trails=2, Spawn Interval=0)
  ]
  particleSpawnScript [
   &FN_InitializeAnimTrailBoneIndecies(Second Bone or Socket Index=0)
   &FN_InitializeAnimTrailBoneIndecies(First Bone or Socket Index=1) {module=FN_InitializeAnimTrailBoneIndecies001}
   &FN_InitializeAnimTrail(
    AnimTrailWidth=80
    Color=&ColorFromCurve(
     ColorCurve=DataInterface() {propertyValues {
       AlphaCurve {keys [
         <at 0.0736 0 RCIM_Cubic>
         {arriveTangent=1.4506, interpMode=RCIM_Cubic, leaveTangent=1.4506, time=0.2857, value=1}
         <at 0.7629 1>
         <at 0.987 0>
        ]}
       BlueCurve {keys [<at 0.4951 1 RCIM_Cubic>]}
       GreenCurve {keys [<at 0.4951 1 RCIM_Cubic>]}
       RedCurve {keys [<at 0.4951 1 RCIM_Cubic>]}
      }}
     CurveIndex=$Emitter.NormalizedLoopAge
    )
    Lifetime=0.6
    SkeletalMesh=$System.Skeletal Mesh
    Write Material Random=false
   )
   &AddVelocity(Velocity=&FN_FrontEndVector3Modifier(Gameplay=<0 0 0>, Front End=<0 0 0>))
   &CameraOffset()
  ]
  particleUpdateScript [
   &ScaleColor(
    Scale Alpha=&Multiply_Float(
     A=&FloatFromCurve(
      FloatCurve=DataInterface() {propertyValues {Curve {keys [{interpMode=RCIM_Cubic, tangentMode=RCTM_User, time=0, value=1}, {interpMode=RCIM_Cubic, tangentMode=RCTM_User, time=1, value=0}]}}}
      Scale Curve=&FloatFromCurve(
       FloatCurve=DataInterface() {propertyValues {Curve {keys [
           {interpMode=RCIM_Cubic, tangentMode=RCTM_User, time=0.4, value=0}
           {interpMode=RCIM_Cubic, tangentMode=RCTM_User, time=0.5, value=1}
           {arriveTangent=0.0018, interpMode=RCIM_Cubic, leaveTangent=0.0018, tangentMode=RCTM_User, time=0.9, value=1}
           <at 1.2 0 RCIM_Cubic>
          ]}}}
       CurveIndex=$Emitter.Age
       Scale Curve=&FN_FrontEndFloatModifier(
        Gameplay=&MaskFloatBySpawnGroup(Masked Float=1.5, Passthrough Float=1)
        Front End=&MaskFloatBySpawnGroup(Masked Float=2, Passthrough Float=0.25)
       )
      )
     )
     B=&FNTimeOfDayScalar(Night Scale=0.6, Morning Scale=0.8, Evening Scale=0.8)
    )
    Scale RGB=&MakeVectorFromLinearColorRGB(Color=&SelectLinearColorFromArray(
      Array Sampling Mode=::NewEnumerator1
      Direct Array Index=&SpawnGroup()
      Color Selection Array=DataInterface() {propertyValues {ColorData [<rgba 0.1 0.1 0.005 1>, <rgba 0.4 0.0106 0.004 1>]}}
      Random Seed=?
     ))
    ScaleA=true
    ScaleRGB=true
   ) {module=ScaleColor001}
   &SolveForcesAndVelocity(Clamp Velocity=false, Limit Acceleration=false)
   &FN_UpdateAnimTrail(Locked Particle RibbonWidth=1, SkeletalMesh=$System.Skeletal Mesh) {module=FN_UpdateAnimTrail001}
   /SetVariables(Particles.RibbonU0Override=$Particles.NormalizedAge) {isSetParameters=true}
   &ScaleRibbonWidth(Ribbon Width Scale=&FloatFromCurve(FloatCurve=DataInterface() {propertyValues {Curve {keys [{interpMode=RCIM_Cubic, tangentMode=RCTM_User, time=0, value=1}, {interpMode=RCIM_Cubic, tangentMode=RCTM_User, time=1, value=0}]}}}))
  ]
  rendererClasses [&NiagaraRibbonRendererProperties]
  renderers [{rendererClass=&NiagaraRibbonRendererProperties}]
 }
 Smoke_Puff {
  emitterUpdateScript [
   &EmitterState(Life Cycle Mode=::NewEnumerator1, Loop Behavior=::NewEnumerator1, Loop Delay=2.2, Loop Duration Mode=::NewEnumerator1, UseLoopDelay=true)
   &SpawnRate(Spawn Probability=0.5, SpawnRate=25, Use Spawn Probability=true)
   &SpawnRate(
    Spawn Probability=&FloatFromCurve(
     FloatCurve=DataInterface() {propertyValues {Curve {keys [{interpMode=RCIM_Cubic, tangentMode=RCTM_User, time=0, value=0}, {interpMode=RCIM_Cubic, tangentMode=RCTM_User, time=1, value=1}]}}}
     CurveIndex=$System.Timeline.Time
    )
    SpawnRate=5
    Use Spawn Probability=true
   ) {module=SpawnRate001}
  ]
  particleSpawnScript [
   /SetVariables(Particles.PerParticleRandom:NiagaraFloat=&RandomRangeFloat(Minimum=0, Maximum=1)) {isSetParameters=true, module=SetVariables_#3}
   &FN_SetSpawnGroup(Number of Spawn Groups=2)
   &InitializeParticle(
    Mesh Scale=&Multiply_VectorByFloat(Vector=<2 1 1>, Float=1.5)
    Mesh Scale Mode=::NewEnumerator4
    UsePositionOffset=false
    Write Material Random=true
   )
   &InitialMeshOrientation(
    Mesh Orientation Mode=::NewEnumerator7
    Rotation=&MakeVector(X=&RandomRangeFloat(Minimum=0, Maximum=1), Y=0, Z=0)
    Use Rotation Vector=true
   )
  ]
  particleUpdateScript [
   &ParticleState()
   &ScaleMeshSize(Scale Factor=&VectorFromFloat(Value=&FloatFromCurve(
      FloatCurve=DataInterface() {propertyValues {Curve {keys [
          {arriveTangent=0.0166, interpMode=RCIM_Cubic, leaveTangent=0.0166, tangentMode=RCTM_User, time=0, value=0.125}
          {arriveTangent=0.5524, interpMode=RCIM_Cubic, leaveTangent=0.5524, tangentMode=RCTM_User, time=1, value=1.475}
         ]}}}
      Scale Curve=&FloatFromCurve(
       FloatCurve=DataInterface() {propertyValues {Curve {keys [{interpMode=RCIM_Cubic, tangentMode=RCTM_User, time=0, value=1}, {interpMode=RCIM_Cubic, tangentMode=RCTM_User, time=1, value=1.2}]}}}
       CurveIndex=$System.Timeline.Time
      )
     )))
   &ScaleColor(
    Scale Alpha=&FloatFromCurve(FloatCurve=DataInterface() {propertyValues {Curve {keys [
         {arriveTangent=0.0198, interpMode=RCIM_Cubic, leaveTangent=0.0198, tangentMode=RCTM_User, time=0.2, value=0}
         {arriveTangent=0.0061, interpMode=RCIM_Cubic, leaveTangent=0.0061, tangentMode=RCTM_User, time=0.4, value=1}
         {interpMode=RCIM_Cubic, tangentMode=RCTM_User, time=0.5, value=1}
         {arriveTangent=-1.2517, interpMode=RCIM_Cubic, leaveTangent=-1.2517, tangentMode=RCTM_User, time=1, value=0}
        ]}}})
    Scale RGB=&MakeVectorFromLinearColorRGB(Color=&ColorFromCurve(ColorCurve=DataInterface() {propertyValues {
        AlphaCurve {keys [<at 0 1 RCIM_Cubic>]}
        BlueCurve {keys [<at 0 0.1 RCIM_Cubic>,<at 0.3108 1>,<at 0.8079 0.4>]}
        GreenCurve {keys [<at 0 0.4 RCIM_Cubic>,<at 0.3108 1>,<at 0.8079 0.1993>]}
        RedCurve {keys [<at 0 1 RCIM_Cubic>,<at 0.3108 1>,<at 0.8079 0.12>]}
       }}))
    ScaleA=true
    ScaleRGB=true
   )
   &ScaleColor(
    Color Value To Scale=$Particles.Color
    Scale RGB=&MakeVectorFromLinearColorRGB(Color=$System.TOD Color)
    ScaleA=true
    ScaleRGB=true
   ) {module=ScaleColor003}
   &DynamicMaterialParameters(
    Index 0 Param 1=&FloatFromCurve(FloatCurve=DataInterface() {propertyValues {Curve {keys [
         {arriveTangent=0.0074, interpMode=RCIM_Cubic, leaveTangent=0.0074, tangentMode=RCTM_User, time=0, value=0.2}
         {arriveTangent=0.4016, interpMode=RCIM_Cubic, leaveTangent=0.4016, tangentMode=RCTM_User, time=1, value=1}
        ]}}})
    Module.Index0Param0Write=true
    Module.Index0Param1Write=true
    Module.Index0Param2Write=true
    Module.Index0Param3Write=true
   )
   &self:FollowRibbon(
    Alpha:NiagaraFloat=&FloatFromCurve(FloatCurve=DataInterface() {propertyValues {Curve {keys [
         {arriveTangent=2.4691, interpMode=RCIM_Cubic, leaveTangent=2.4691, tangentMode=RCTM_User, time=0, value=0}
         {arriveTangent=-0.0157, interpMode=RCIM_Cubic, leaveTangent=-0.0157, tangentMode=RCTM_User, time=0.9, value=0.8}
        ]}}})
    ParticlePerRibbon=25
    RibbonAmount=2
    RibbonNumber:NiagaraInt32=&SpawnGroup()
    RibbonReader=DataInterface() {propertyValues {EmitterBinding {bindingMode=Other, emitterName=Smoke_Ribbon}}}
   ) {module=FollowRibbon}
   /SetVariables(Particles.Position=&AddVectorToPosition(
     Position=$Output.FollowRibbon.RibbonPos
     Vector=&Multiply_VectorByFloat(
      Vector=&RandomVector(Vector Scale=1)
      Float=&FloatFromCurve(
       FloatCurve=DataInterface() {propertyValues {Curve {keys [
           {arriveTangent=-0.0262, interpMode=RCIM_Cubic, leaveTangent=-0.0262, tangentMode=RCTM_User, time=0, value=0}
           {arriveTangent=1.5356, interpMode=RCIM_Cubic, leaveTangent=1.5356, tangentMode=RCTM_User, time=1, value=1}
          ]}}}
       Scale Curve=&RandomRangeFloat(Minimum=0, Maximum=50)
      )
     )
    )) {isSetParameters=true}
   &UpdateMeshOrientation(
    Facing Direction=$Output.FollowRibbon.DirectionVector
    Facing Interpolation=false
    Orientation Method=::NewEnumerator1
    Override Quat=false
    TransformFacing=false
    TransformUp=false
   ) {module=UpdateMeshOrientation001}
   /SetVariables(Particles.MeshOrientation=&MultiplyQuaternion(
     Quaternion A=$Particles.MeshOrientation
     Quaternion B=&MakeQuatFromAxisAngle(
      Axis=<1 0 0>
      AngleInDegrees=&FloatFromCurve(
       FloatCurve=DataInterface() {propertyValues {Curve {keys [
           {arriveTangent=2.4202, interpMode=RCIM_Cubic, leaveTangent=2.4202, tangentMode=RCTM_User, time=0, value=0}
           {interpMode=RCIM_Cubic, tangentMode=RCTM_User, time=0.8, value=1}
          ]}}}
       Scale Curve=&Multiply_Float(
        A=&RandomRangeFloat(Minimum=-10, Maximum=10)
        B=&Lerp_Float(A=1, B=1.2, Alpha=$System.Timeline.Time)
       )
      )
     )
    )) {isSetParameters=true}
  ]
  rendererClasses [&NiagaraMeshRendererProperties]
  renderers [{rendererClass=&NiagaraMeshRendererProperties, rendererIndex=0}]
 }
 Smoke_Ribbon {
  emitterUpdateScript [
   &EmitterState(Life Cycle Mode=::NewEnumerator1, Loop Behavior=::NewEnumerator1, Loop Duration Mode=::NewEnumerator1, UseLoopDelay=false)
   &SpawnBurst_Instantaneous(Spawn Count=50, Use Spawn Probability=false, UseLoopCountLimit=true)
  ]
  particleSpawnScript [
   &PartitionParticles(Write Partition to Ribbon ID=::NewEnumerator0)
   &InitializeParticle(
    Lifetime=0.2
    Position Offset=<0 0 25>
    Ribbon Width=&FN_FrontEndFloatModifier(Gameplay=100, Front End=75)
    Ribbon Width Mode=::NewEnumerator1
    Sprite Size Mode=::NewEnumerator3
    Uniform Sprite Size=5
    UsePositionOffset=false
    Write Material Random=true
   )
   /SetVariables(
    Particles.RibbonID=/Niagara/DynamicInputs/Meshes/MakeRibbonIDFromSpawnGroup()
    Particles.RibbonLinkOrder:NiagaraFloat=&ReturnNormalizedExecIndex()
    Particles.SpawnGroup=$Particles.PartitionParticles.Partition
   ) {isSetParameters=true}
   &DynamicMaterialParameters(
    Index 0 Param 1=0.3
    Index 0 Param 2=&SelectFloatFromArray(Direct Array Index=$Particles.PartitionParticles.Partition, Float Selection Array=DataInterface() {propertyValues {FloatData [0,0.25]}})
    Module.Index0Param0Write=true
    Module.Index0Param1Write=true
    Module.Index0Param2Write=true
    Module.Index0Param3Write=true
   )
   /SetVariables() {isSetParameters=true}
  ]
  particleUpdateScript [
   &ParticleState(Kill Particles When Lifetime Has Elapsed=false, Let Infinitely Lived Particles Die When Emitter Deactivates=true)
   &SkeletalMeshLocation(
    Bone / Socket Index=&SelectIntFromArray(Direct Array Index=$Particles.PartitionParticles.Partition, Int Selection Array=DataInterface() {propertyValues {IntData [0,1]}})
    MaskSpawnGroup=false
    Mesh Sampling Type=::NewEnumerator5
    Sampled Position Offset Coordinate Space:ENiagaraSkelMeshTransforms=::NewEnumerator1
    Skeletal Mesh=$System.Skeletal Mesh
    Socket Sampling Mode=::NewEnumerator5
   ) {module=SkeletalMeshLocation001}
   /SetVariables(Particles.Position=&LerpPosition(
     A=$Particles.Position
     B=&AddVectorToPosition(
      Position=$Particles.Position
      Vector=&MakeVector(
       X=&FloatFromCurve(
        FloatCurve=DataInterface() {propertyValues {Curve {keys [<at 0.3 0 RCIM_Cubic>, {arriveTangent=2.6447, interpMode=RCIM_Cubic, leaveTangent=2.6447, tangentMode=RCTM_User, time=1, value=1}]}}}
        CurveIndex=&ReturnNormalizedExecIndex()
        Scale Curve=&Multiply_Float(
         A=&MakeFloatFromVector(Vector=$System.InterpolateOverTime.InterpolateOverTime_V3_.Moving Average, Channel=::NewEnumerator0)
         B=&Lerp_Float(A=0, B=300, Alpha=$System.InterpolateOverTime001.InterpolateOverTime_Float.Moving Average)
        )
       )
       Y=&FloatFromCurve(
        FloatCurve=DataInterface() {propertyValues {Curve {keys [{tangentMode=RCTM_User, time=0, value=0}, {tangentMode=RCTM_User, time=1, value=1}]}}}
        CurveIndex=&ReturnNormalizedExecIndex()
        Scale Curve=&Multiply_Float(
         A=&FloatFromCurve(
          FloatCurve=DataInterface() {propertyValues {Curve {keys [
              {arriveTangent=3.9602, interpMode=RCIM_Cubic, leaveTangent=3.9602, tangentMode=RCTM_User, time=2, value=0}
              {arriveTangent=-0.0348, interpMode=RCIM_Cubic, leaveTangent=-0.0348, tangentMode=RCTM_User, time=2.5, value=1}
             ]}}}
          CurveIndex=$Emitter.Age
         )
         B=&FN_FrontEndFloatModifier(
          Gameplay=&Lerp_Float(A=-350, B=-500, Alpha=$System.Timeline.NormalizedTime)
          Front End=&Lerp_Float(A=250, B=350, Alpha=$System.Timeline002.NormalizedTime)
         )
        )
       )
       Z=0
      )
     )
     Alpha=&ReturnNormalizedExecIndex()
    )) {isSetParameters=true}
   &ScaleColor(
    Scale Alpha=&FloatFromCurve(
     FloatCurve=DataInterface() {propertyValues {Curve {keys [
         <at 0 0 RCIM_Cubic>
         {arriveTangent=0.0209, interpMode=RCIM_Cubic, leaveTangent=0.0209, tangentMode=RCTM_User, time=0.5, value=1}
         {arriveTangent=-0.0049, interpMode=RCIM_Cubic, leaveTangent=-0.0049, tangentMode=RCTM_User, time=0.85, value=1}
         {arriveTangent=-10.6017, interpMode=RCIM_Cubic, leaveTangent=-10.6017, tangentMode=RCTM_User, time=1, value=0}
        ]}}}
     CurveIndex=$Particles.Initial.RibbonLinkOrder
     Scale Curve=&FN_FrontEndFloatModifier(
      Gameplay=&FloatFromCurve(
       FloatCurve=DataInterface() {propertyValues {Curve {keys [
           {arriveTangent=4.1911, interpMode=RCIM_Cubic, leaveTangent=4.1911, tangentMode=RCTM_User, time=2, value=0}
           {arriveTangent=0.0589, interpMode=RCIM_Cubic, leaveTangent=0.0589, tangentMode=RCTM_User, time=2.5, value=1}
          ]}}}
       CurveIndex=$Emitter.Age
      )
      Front End=&Lerp_Float(A=0, B=1, Alpha=$System.Timeline002.NormalizedTime)
     )
    )
    Scale RGB=&MakeVectorFromLinearColorRGB(Color=&ColorFromCurve(
      ColorCurve=DataInterface() {propertyValues {
        AlphaCurve {keys [<at 0 1 RCIM_Cubic>]}
        BlueCurve {keys [<at 0.5939 1>,<at 1 0.4 RCIM_Cubic>]}
        GreenCurve {keys [<at 0.5939 1>,<at 1 0.1993 RCIM_Cubic>]}
        RedCurve {keys [<at 0.5939 1>,<at 1 0.12 RCIM_Cubic>]}
       }}
      CurveIndex=$Particles.Initial.RibbonLinkOrder
     ))
    ScaleA=true
    ScaleRGB=true
   ) {module=ScaleColor001}
   &ScaleColor(
    Color Value To Scale=$Particles.Color
    Scale RGB=&MakeVectorFromLinearColorRGB(Color=&LerpLinearColors(
      StartColor=&LerpLinearColors(StartColor=<rgba 1.5 0.55 0.075 1>, EndColor=<rgba 5 2.75 0.5 1>, LerpFactor=$System.Timeline.Time)
      EndColor=<rgba 1 1 1 1>
      LerpFactor=&FloatFromCurve(
       FloatCurve=DataInterface() {propertyValues {Curve {keys [{interpMode=RCIM_Cubic, tangentMode=RCTM_User, time=0.3, value=0}, {interpMode=RCIM_Cubic, tangentMode=RCTM_User, time=0.6, value=1}]}}}
       CurveIndex=$Particles.Initial.RibbonLinkOrder
      )
     ))
    ScaleA=true
    ScaleRGB=true
   ) {module=ScaleColor002}
   &ScaleColor(
    Color Value To Scale=$Particles.Color
    Scale RGB=&MakeVectorFromLinearColorRGB(Color=$System.TOD Color)
    ScaleA=true
    ScaleRGB=true
   ) {module=ScaleColor003}
   &ScaleRibbonWidth(Ribbon Width Scale=&FloatFromCurve(
     FloatCurve=DataInterface() {propertyValues {Curve {keys [
         {arriveTangent=0.0168, interpMode=RCIM_Cubic, leaveTangent=0.0168, tangentMode=RCTM_User, time=0.25, value=0.1}
         {arriveTangent=0.9034, interpMode=RCIM_Cubic, leaveTangent=0.9034, tangentMode=RCTM_User, time=0.9, value=1.125}
        ]}}}
     CurveIndex=$Particles.Initial.RibbonLinkOrder
     Scale Curve=&FloatFromCurve(
      FloatCurve=DataInterface() {propertyValues {Curve {keys [
          {arriveTangent=-0.008, interpMode=RCIM_Cubic, leaveTangent=-0.008, tangentMode=RCTM_User, time=2, value=0}
          {arriveTangent=-0.0654, interpMode=RCIM_Cubic, leaveTangent=-0.0654, tangentMode=RCTM_User, time=2.4, value=1.5}
          {interpMode=RCIM_Cubic, tangentMode=RCTM_User, time=2.7, value=1}
         ]}}}
      CurveIndex=$Emitter.Age
      Scale Curve=&FloatFromCurve(
       FloatCurve=DataInterface() {propertyValues {Curve {keys [{interpMode=RCIM_Cubic, tangentMode=RCTM_User, time=0, value=1}, {interpMode=RCIM_Cubic, tangentMode=RCTM_User, time=1, value=1.5}]}}}
       CurveIndex=$System.Timeline.Time
      )
     )
    ))
   /SetVariables(Particles.Offset X:NiagaraFloat=&Add_Float(
     A=$Particles.Offset X
     B=&FN_FrontEndFloatModifier(
      Gameplay=&Lerp_Float(
       A=&FloatFromCurve(
        FloatCurve=DataInterface() {propertyValues {Curve {keys [
            {arriveTangent=0.0005, interpMode=RCIM_Cubic, leaveTangent=0.0005, tangentMode=RCTM_User, time=2, value=1}
            {arriveTangent=-0.0348, interpMode=RCIM_Cubic, leaveTangent=-0.0348, tangentMode=RCTM_User, time=3.5, value=0}
           ]}}}
        CurveIndex=$Emitter.Age
        Scale Curve=-0.02
       )
       B=-0.01
       Alpha=$System.Timeline.Time
      )
      Front End=-0.01
     )
    )) {isSetParameters=true}
   &DynamicMaterialParameters(
    Index 0 Param 2=&Add_Float(
     A=&SelectFloatFromArray(Direct Array Index=$Particles.PartitionParticles.Partition, Float Selection Array=DataInterface() {propertyValues {FloatData [0,0.25]}})
     B=$Particles.Offset X
    )
    Module.Index0Param0Write=false
    Module.Index0Param1Write=true
    Module.Index0Param2Write=false
    Module.Index0Param3Write=false
   ) {module=DynamicMaterialParameters001}
  ]
  rendererClasses [&NiagaraRibbonRendererProperties]
  renderers [{rendererClass=&NiagaraRibbonRendererProperties}]
 }
 SpeedLines {
  emitterUpdateScript [
   &EmitterState(Life Cycle Mode=::NewEnumerator1, Loop Behavior=::NewEnumerator1, Loop Duration Mode=::NewEnumerator1, UseLoopDelay=false)
   &SpawnRate(
    SpawnRate=&FN_FrontEndFloatModifier(Gameplay=&Lerp_Float(A=0, B=100, Alpha=$System.Timeline.NormalizedTime), Front End=50)
    Use Spawn Probability=true
   ) {module=SpawnRate001}
  ]
  particleSpawnScript [
   &InitializeParticle(
    Color=&MakeLinearColorFromVectorAndFloat(
     "Vector (RGB)"=<1 1 1>
     "Float (Alpha)"=&Multiply_Float(
      A=&UniformRangedFloat(Minimum=0.1, Maximum=0.18)
      B=&FN_FrontEndFloatModifier(
       Gameplay=1
       Front End=&FloatFromCurve(
        FloatCurve=DataInterface() {propertyValues {Curve {keys [{interpMode=RCIM_Cubic, tangentMode=RCTM_User, time=2.15, value=0}, {interpMode=RCIM_Cubic, tangentMode=RCTM_User, time=2.35, value=1}]}}}
        CurveIndex=$Emitter.Age
        Scale Curve=0.5
       )
      )
     )
    )
    Lifetime Max=0.125
    Lifetime Min=0.1
    Lifetime Mode=::NewEnumerator1
    Sprite Size Max=<3 10>
    Sprite Size Min=<2 2>
    Sprite Size Mode=::NewEnumerator2
    UsePositionOffset=false
    Write Material Random=true
   )
   &TorusLocation(
    Align Torus With Axis=true
    Handle Radius=0
    Large Radius:NiagaraFloat=&FN_FrontEndFloatModifier(
     Gameplay=&RandomRangeFloat(Minimum=200, Maximum=300)
     Front End=&RandomRangeFloat(Minimum=25, Maximum=100)
    )
    MaskSpawnGroup=false
    Module.Override Randomness=false
    Torus Axis=<0 1 0>
    Torus Origin:NiagaraPosition=&Add_Vector(
     A=Linked() {linkedVariable {name=Particles.Position, type=NiagaraPosition}}
     B=&FN_FrontEndVector3Modifier(Gameplay=<0 0 0>, Front End=<0 -200 0>)
    )
   ) {enabled=false}
   &ShapeLocation(
    Handle Radius=0
    Large Radius=&FN_FrontEndFloatModifier(
     Gameplay=&RandomRangeFloat(Minimum=200, Maximum=300)
     Front End=&RandomRangeFloat(Minimum=25, Maximum=120)
    )
    MaskSpawnGroup=false
    Non Uniform Scale=&FN_FrontEndVector3Modifier(Gameplay=<1 1 1>, Front End=<1 0.6 1>)
    Offset=&FN_FrontEndVector3Modifier(Gameplay=<0 0 0>, Front End=<0 -200 -60>)
    Offset Mode=::NewEnumerator0
    Override Seed=false
    Rotation Axis=<0 1 0>
    Rotation Mode=::NewEnumerator1
    Shape Primitive=::NewEnumerator3
    UseOwnerScale=true
   )
   /SetVariables(Particles.Position=&Multiply_Vector(
     A=Linked() {linkedVariable {name=Particles.Position, type=NiagaraPosition}}
     B=&FN_FrontEndVector3Modifier(Gameplay=<1 1 1>, Front End=<1 1 0.5>)
    )) {enabled=false, isSetParameters=true}
   &AddVelocityFromPoint(
    Module.Use Falloff Distance=false
    Origin Offset=&FN_FrontEndVector3Modifier(Gameplay=<0 500 0>, Front End=<0 -500 -60>)
    Velocity Strength=&Multiply_Float(A=600, B=&FN_FrontEndFloatModifier(Gameplay=1, Front End=0.5))
   )
   &AddVelocity(Velocity=&Multiply_Vector(
     A=&RandomRangeVector(Minimum=<0 300 0>, Maximum=<0 700 0>)
     B=&MakeVector(X=1, Y=&FN_FrontEndFloatModifier(Gameplay=-1, Front End=0.001), Z=1)
    ))
  ]
  particleUpdateScript [
   &SolveForcesAndVelocity(Clamp Velocity=false, Limit Acceleration=false)
   &ParticleState()
   &ScaleSpriteSize(Scale Factor=&Vector2DFromCurve(
     Vector2Curve=DataInterface() {propertyValues {XCurve {keys [<at 0 0.5 RCIM_Cubic>,<at 0.5 1 RCIM_Cubic>]}, YCurve {keys [<at 0 0 RCIM_Cubic>,<at 0.5 1 RCIM_Cubic>]}}}
     ScaleCurve=&Vector2DFromFloat(Value=1)
    ))
   &ScaleSpriteSizeBySpeed(
    Max Scale Factor=&FN_FrontEndVector2Modifier(Front End:Vector2f=<0.75 30>, Gameplay:Vector2f=<1 50>)
    Min Scale Factor=<0.01 0.05>
    Velocity Threshold=1
   )
   &ScaleColor(
    Scale Alpha=&FN_FrontEndFloatModifier(
     Gameplay=&Lerp_Float(A=0, B=1, Alpha=$System.Timeline.NormalizedTime)
     Front End=&RandomRangeFloat(Minimum=0.5, Maximum=0.8)
    )
    Scale RGB=&VectorFromFloat(Value=1)
    ScaleA=true
    ScaleRGB=true
   )
  ]
  rendererClasses [&NiagaraSpriteRendererProperties]
  renderers [{rendererClass=&NiagaraSpriteRendererProperties}]
 }
 Thruster_Flame {
  emitterSpawnScript [/SetVariables(Emitter.CapsuleGlowDirection:Vector3f=<0 1 0>) {isSetParameters=true, module=SetVariables_#4}]
  emitterUpdateScript [
   &EmitterState(
    Inactive Response=::NewEnumerator1
    Life Cycle Mode=::NewEnumerator1
    Loop Behavior=::NewEnumerator1
    Loop Delay=0.25
    Loop Duration Mode=::NewEnumerator1
    UseLoopDelay=true
   )
   &SpawnBurst_Instantaneous(Spawn Count=2, Spawn Time=0.4, Use Spawn Probability=false, UseLoopCountLimit=false)
  ]
  particleSpawnScript [
   &InitializeParticle(
    Color=<rgba 1 0.5833 0 4>
    Mesh Scale=&Multiply_VectorByFloat(Vector=<1 1.25 1>, Float=&FN_FrontEndFloatModifier(Gameplay=1.6, Front End=1.3))
    Mesh Scale Mode=::NewEnumerator4
    Sprite Size Mode=::NewEnumerator3
    Uniform Sprite Size=&FN_FrontEndFloatModifier(Gameplay=30, Front End=25)
    UsePositionOffset=false
    Write Material Random=true
   )
   &FN_SetSpawnGroup(Number of Spawn Groups=2)
   &InitialMeshOrientation(
    Mesh Axis To Orient=<0 1 0>
    Mesh Orientation Mode=::NewEnumerator2
    Orientation Vector=&MaskVectorBySpawnGroup(Masked Vector=<0 1 0>, Passthrough Vector=<0 -1 0>)
    Use Rotation Vector=false
   )
   /SetVariables(
    Particles.GlowCameraOffset:NiagaraFloat=&FN_FrontEndFloatModifier(Gameplay=10, Front End=0)
    Particles.GlowColor=<rgba 10 1.6667 0 1>
    Particles.GlowPosition:NiagaraPosition=&AddVectorToPosition(Position=$Particles.Position, Vector=<0 0 0>)
    Particles.GlowSize=&Vector2DFromFloat(Value=50)
   ) {isSetParameters=true}
  ]
  particleUpdateScript [
   &ParticleState(Kill Particles When Lifetime Has Elapsed=false)
   &SkeletalMeshLocation(
    Bone / Socket Index=&SpawnGroup()
    MaskSpawnGroup=false
    Mesh Sampling Type=::NewEnumerator5
    Sampled Position Offset=&Multiply_VectorByFloat(Vector=&MaskVectorBySpawnGroup(Masked Vector=<0 1 0>, Passthrough Vector=<0 -1 0>), Float=0)
    Skeletal Mesh=$System.Skeletal Mesh
    Socket Sampling Mode=::NewEnumerator5
   )
   &ScaleMeshSize(Scale Factor=&VectorFromFloat(Value=&FloatFromCurve(
      FloatCurve=DataInterface() {propertyValues {Curve {keys [
          {arriveTangent=-0.0286, interpMode=RCIM_Cubic, leaveTangent=-0.0286, tangentMode=RCTM_User, time=2.1, value=0}
          {arriveTangent=-0.0018, interpMode=RCIM_Cubic, leaveTangent=-0.0018, tangentMode=RCTM_User, time=2.2, value=1.25}
          {arriveTangent=0.0249, interpMode=RCIM_Cubic, leaveTangent=0.0249, tangentMode=RCTM_User, time=3, value=1}
         ]}}}
      CurveIndex=$Emitter.Age
     )))
   &ScaleMeshSize(
    Initial Mesh Scale:Vector3f=$Particles.Scale
    Scale Factor=&Lerp_Vector(A=<1 1 1>, B=<1.2 1.5 1.2>, Alpha=$System.Timeline.Time)
   ) {module=ScaleMeshSize001}
   /SetVariables(Particles.Color=&LerpLinearColors(
     StartColor=<rgba 1 0.1667 0 4>
     EndColor=<rgba 1 0.5 0 4>
     LerpFactor=&ClampFloat(Float=&Add_Float(
       A=&FloatFromCurve(
        FloatCurve=DataInterface() {propertyValues {Curve {keys [
            {arriveTangent=25.4307, interpMode=RCIM_Cubic, leaveTangent=25.4307, tangentMode=RCTM_User, time=2.1, value=0}
            {arriveTangent=-0.0018, interpMode=RCIM_Cubic, leaveTangent=-0.0018, tangentMode=RCTM_User, time=2.2, value=1}
            {arriveTangent=0.0249, interpMode=RCIM_Cubic, leaveTangent=0.0249, tangentMode=RCTM_User, time=3, value=0}
           ]}}}
        CurveIndex=$Emitter.Age
       )
       B=&FloatFromCurve(
        FloatCurve=DataInterface() {propertyValues {Curve {keys [{tangentMode=RCTM_User, time=0, value=0}, {tangentMode=RCTM_User, time=1, value=1}]}}}
        CurveIndex=$System.Timeline.Time
       )
      ))
    )) {isSetParameters=true}
   &ScaleColor(
    Color Value To Scale=$Particles.Color
    Scale RGB=&VectorFromFloat(Value=&FloatFromCurve(
      FloatCurve=DataInterface() {propertyValues {Curve {keys [
          {arriveTangent=0.0378, interpMode=RCIM_Cubic, leaveTangent=0.0378, tangentMode=RCTM_User, time=2, value=0}
          {arriveTangent=-0.0018, interpMode=RCIM_Cubic, leaveTangent=-0.0018, tangentMode=RCTM_User, time=2.3, value=2}
          {arriveTangent=0.0249, interpMode=RCIM_Cubic, leaveTangent=0.0249, tangentMode=RCTM_User, time=3, value=1.5}
         ]}}}
      CurveIndex=$Emitter.Age
      Scale Curve=&FloatFromCurve(
       FloatCurve=DataInterface() {propertyValues {Curve {keys [{interpMode=RCIM_Cubic, tangentMode=RCTM_User, time=0, value=1}, {interpMode=RCIM_Cubic, tangentMode=RCTM_User, time=1, value=1.5}]}}}
       CurveIndex=$System.Timeline.Time
       Scale Curve=&FNTimeOfDayScalar(Night Scale=0.8, Morning Scale=1, Day Scale=1.5, Evening Scale=1)
      )
     ))
    ScaleA=true
    ScaleRGB=true
   )
   /SetVariables(
    Particles.Thruster Inner Color:LinearColor=&MultiplyLinearColor(Linear Color=<rgba 1 0.4 0.2 1>, Scale Factor=&MakeVector4_FromFloat(Float=3))
    Particles.Thruster Inner Size:Vector3f=&Multiply_Vector(A=$Particles.Scale, B=<0.6 0.6 0.6>)
   ) {isSetParameters=true}
   &FN_Light_Attributes(
    Color=$Particles.Color
    Create Color Attribute=true
    Create Position Attribute=true
    Create Radius Attribute=true
    Position=&AddVectorToPosition(Position=$Particles.Position, Vector=&TransformVector(Vector=<0 -30 0>, Destination Space=::Simulation))
    Radius=50
    Scale Color=5
   )
   &DynamicMaterialParameters(
    Index 0 Param 1=&MakeFloatFromVector(
     Vector=&RotateVectorByQuaternion(Vector To Rotate=$Emitter.CapsuleGlowDirection, Quaternion=$Particles.SkeletalMeshLocation.SampledRotation)
     Channel=::NewEnumerator0
    )
    Index 0 Param 2=&MakeFloatFromVector(
     Vector=&RotateVectorByQuaternion(Vector To Rotate=$Emitter.CapsuleGlowDirection, Quaternion=$Particles.SkeletalMeshLocation.SampledRotation)
     Channel=::NewEnumerator1
    )
    Index 0 Param 3=&MakeFloatFromVector(
     Vector=&RotateVectorByQuaternion(Vector To Rotate=$Emitter.CapsuleGlowDirection, Quaternion=$Particles.SkeletalMeshLocation.SampledRotation)
     Channel=::NewEnumerator2
    )
    Index 0 Param 4=&FN_FrontEndFloatModifier(Gameplay=125, Front End=75)
    Module.Param0WriteEnabled=true
    Module.Param1WriteEnabled=true
    Module.Param2WriteEnabled=true
    Module.Param3WriteEnabled=true
   )
   /SetVariables(Particles.GlowColor=&MultiplyLinearColor(
     Linear Color=&LerpLinearColors(
      StartColor=<rgba 1 0.1667 0 1>
      EndColor=<rgba 1 0.4167 0 4>
      LerpFactor=&ClampFloat(Float=&Add_Float(
        A=&FloatFromCurve(
         FloatCurve=DataInterface() {propertyValues {Curve {keys [
             {arriveTangent=25.4307, interpMode=RCIM_Cubic, leaveTangent=25.4307, tangentMode=RCTM_User, time=2.1, value=0}
             {arriveTangent=-0.0018, interpMode=RCIM_Cubic, leaveTangent=-0.0018, tangentMode=RCTM_User, time=2.2, value=1}
             {arriveTangent=0.0249, interpMode=RCIM_Cubic, leaveTangent=0.0249, tangentMode=RCTM_User, time=3, value=0}
            ]}}}
         CurveIndex=$Emitter.Age
        )
        B=&FloatFromCurve(
         FloatCurve=DataInterface() {propertyValues {Curve {keys [{tangentMode=RCTM_User, time=0, value=0}, {tangentMode=RCTM_User, time=1, value=1}]}}}
         CurveIndex=$System.Timeline.Time
        )
       ))
     )
     Scale Factor=&MakeFloat4FromVectorAndFloat(
      "Vector (RGB)"=&VectorFromFloat(Value=&MakeFloatFromVector4(
        Vector 4=&MakeVector4_FromFloat(Float=&FloatFromCurve(
          FloatCurve=DataInterface() {propertyValues {Curve {keys [
              {arriveTangent=0.0378, interpMode=RCIM_Cubic, leaveTangent=0.0378, tangentMode=RCTM_User, time=2, value=0}
              {arriveTangent=-0.0018, interpMode=RCIM_Cubic, leaveTangent=-0.0018, tangentMode=RCTM_User, time=2.3, value=1.5}
              {arriveTangent=0.0249, interpMode=RCIM_Cubic, leaveTangent=0.0249, tangentMode=RCTM_User, time=3, value=1}
             ]}}}
          CurveIndex=$Emitter.Age
          Scale Curve=&FloatFromCurve(
           FloatCurve=DataInterface() {propertyValues {Curve {keys [{interpMode=RCIM_Cubic, tangentMode=RCTM_User, time=0, value=1}, {interpMode=RCIM_Cubic, tangentMode=RCTM_User, time=1, value=1.5}]}}}
           CurveIndex=$System.Timeline.Time
          )
         ))
        Channel=::NewEnumerator0
       ))
      "Float (Alpha)"=1
     )
    )) {isSetParameters=true}
  ]
  rendererClasses [
   &NiagaraSpriteRendererProperties
   &NiagaraMeshRendererProperties
   &NiagaraLightRendererProperties
  ]
  renderers [
   {rendererClass=&NiagaraSpriteRendererProperties}
   {rendererClass=&NiagaraMeshRendererProperties, rendererIndex=1}
   {rendererClass=&NiagaraMeshRendererProperties, rendererIndex=2}
   {rendererClass=&NiagaraSpriteRendererProperties, rendererIndex=3}
   {rendererClass=&NiagaraLightRendererProperties, rendererIndex=4}
  ]
 }
 Thruster_Spikes {
  emitterSpawnScript [/SetVariables(Emitter.CapsuleGlowDirection:Vector3f=<0 1 0>) {isSetParameters=true, module=SetVariables_#4}]
  emitterUpdateScript [
   &EmitterState(Life Cycle Mode=::NewEnumerator1, Loop Behavior=::NewEnumerator1, Loop Delay=2, Loop Duration Mode=::NewEnumerator1, UseLoopDelay=true)
   &SpawnRate(SpawnRate=6, Use Spawn Probability=false)
   &SpawnBurst_Instantaneous(Spawn Count=3, Use Spawn Probability=false, UseLoopCountLimit=false)
  ]
  particleSpawnScript [
   &InitializeParticle(
    Color=&LerpLinearColors(
     StartColor=<rgba 0.9 0.525 0.45 1>
     EndColor=<rgba 1.5 0.875 0.75 1>
     LerpFactor=&ClampFloat(Float=&Add_Float(
       A=&FloatFromCurve(
        FloatCurve=DataInterface() {propertyValues {Curve {keys [
            {arriveTangent=25.4307, interpMode=RCIM_Cubic, leaveTangent=25.4307, tangentMode=RCTM_User, time=2.1, value=0}
            {arriveTangent=-0.0018, interpMode=RCIM_Cubic, leaveTangent=-0.0018, tangentMode=RCTM_User, time=2.2, value=1}
            {arriveTangent=0.0249, interpMode=RCIM_Cubic, leaveTangent=0.0249, tangentMode=RCTM_User, time=3, value=0}
           ]}}}
        CurveIndex=$Emitter.Age
       )
       B=&FloatFromCurve(
        FloatCurve=DataInterface() {propertyValues {Curve {keys [{tangentMode=RCTM_User, time=0, value=0}, {tangentMode=RCTM_User, time=1, value=1}]}}}
        CurveIndex=$System.Timeline.Time
       )
      ))
    )
    Lifetime=&Lerp_Float(A=0.5, B=1, Alpha=$System.Timeline.Time)
    Mesh Scale=&Multiply_Vector(
     A=&Multiply_VectorByFloat(
      Vector=&MakeVector(X=&RandomRangeFloat(Minimum=5, Maximum=5), Y=1, Z=1)
      Float=&RandomRangeFloat(Minimum=0.25, Maximum=0.3)
     )
     B=&FN_FrontEndVector3Modifier(Gameplay=<0.5 0.6 0.6>, Front End=<0.5 0.45 0.45>)
    )
    Mesh Scale Mode=::NewEnumerator4
    UsePositionOffset=false
    Write Material Random=true
   )
   &FN_SetSpawnGroup(Number of Spawn Groups=2)
   &InitialMeshOrientation(
    Mesh Axis To Orient=<-1 0 0>
    Mesh Orientation Mode=::NewEnumerator2
    Orientation Vector=&MaskVectorBySpawnGroup(Masked Vector=<0 1 0>, Passthrough Vector=<0 -1 0>)
    Rotation=&MakeVector(X=&RandomRangeFloat(Minimum=0, Maximum=1), Y=0, Z=0)
    Use Rotation Vector=true
   )
  ]
  particleUpdateScript [
   &ParticleState()
   &SkeletalMeshLocation(
    Bone / Socket Index=&SpawnGroup()
    MaskSpawnGroup=false
    Mesh Sampling Type=::NewEnumerator5
    Sampled Position Offset=&Multiply_VectorByFloat(Vector=&MaskVectorBySpawnGroup(Masked Vector=<0 1 0>, Passthrough Vector=<0 -1 0>), Float=-5)
    Skeletal Mesh=$System.Skeletal Mesh
    Socket Sampling Mode=::NewEnumerator5
   )
   &ScaleMeshSize(Scale Factor=&Multiply_Vector(
     A=&Lerp_Vector(A=<1 1.2 1.2>, B=<0.6 0.6 0.6>, Alpha=$System.Timeline.Time)
     B=&VectorFromFloat(Value=&FloatFromCurve(
       FloatCurve=DataInterface() {propertyValues {Curve {keys [
           {arriveTangent=1.0872, interpMode=RCIM_Cubic, leaveTangent=1.0872, tangentMode=RCTM_User, time=0, value=1}
           {interpMode=RCIM_Cubic, tangentMode=RCTM_User, time=1, value=1.7}
          ]}}}
       Scale Curve=&FloatFromCurve(
        FloatCurve=DataInterface() {propertyValues {Curve {keys [{interpMode=RCIM_Cubic, tangentMode=RCTM_User, time=0, value=1}, {interpMode=RCIM_Cubic, tangentMode=RCTM_User, time=1, value=1.6}]}}}
        CurveIndex=$System.Timeline.Time
        Scale Curve=&FloatFromCurve(
         FloatCurve=DataInterface() {propertyValues {Curve {keys [
             {arriveTangent=25.4307, interpMode=RCIM_Cubic, leaveTangent=25.4307, tangentMode=RCTM_User, time=2.1, value=0}
             {arriveTangent=-2.8685, interpMode=RCIM_Cubic, leaveTangent=-2.8685, tangentMode=RCTM_User, time=2.2, value=1.8}
             {arriveTangent=0.0249, interpMode=RCIM_Cubic, leaveTangent=0.0249, tangentMode=RCTM_User, time=2.5, value=1}
            ]}}}
         CurveIndex=$Emitter.Age
        )
       )
      ))
    ))
   &ScaleColor(
    Scale Alpha=&FloatFromCurve(FloatCurve=DataInterface() {propertyValues {Curve {keys [
         <at 0 0 RCIM_Cubic>
         {arriveTangent=-0.0089, interpMode=RCIM_Cubic, leaveTangent=-0.0089, tangentMode=RCTM_User, time=0.1, value=1}
         {interpMode=RCIM_Cubic, tangentMode=RCTM_User, time=0.5, value=1}
         {arriveTangent=-4.3084, interpMode=RCIM_Cubic, leaveTangent=-4.3084, tangentMode=RCTM_User, time=1, value=0}
        ]}}})
    Scale RGB=&MakeVectorFromLinearColorRGB(Color=$System.TOD Color)
    ScaleA=true
    ScaleRGB=true
   )
   &DynamicMaterialParameters(
    Index 0 Param 1=&FloatFromCurve(
     FloatCurve=DataInterface() {propertyValues {Curve {keys [{interpMode=RCIM_Cubic, tangentMode=RCTM_User, time=0, value=1}, {interpMode=RCIM_Cubic, tangentMode=RCTM_User, time=1, value=0.5}]}}}
     Scale Curve=&Lerp_Float(A=10, B=15, Alpha=$System.Timeline.Time)
    )
    Index 0 Param 2=&FloatFromCurve(
     FloatCurve=DataInterface() {propertyValues {Curve {keys [{interpMode=RCIM_Cubic, tangentMode=RCTM_User, time=0.5, value=1}]}}}
     Scale Curve=&FN_FrontEndFloatModifier(Gameplay=&Lerp_Float(A=20, B=25, Alpha=$System.Timeline.Time), Front End=20)
    )
    Module.Param0WriteEnabled=true
    Module.Param1WriteEnabled=true
    Module.Param2WriteEnabled=true
    Module.Param3WriteEnabled=true
   )
  ]
  rendererClasses [&NiagaraMeshRendererProperties]
  renderers [{rendererClass=&NiagaraMeshRendererProperties, rendererIndex=0}]
 }
}
summary {
 systemName=NS_Glider_BoomShot_Trail
 emitters [
  {emitterName=Smoke_Ribbon, rendererClasses [&NiagaraRibbonRendererProperties]}
  {emitterName=SpeedLines, rendererClasses [&NiagaraSpriteRendererProperties]}
  {
   emitterName=Thruster_Flame
   rendererClasses [
    &NiagaraSpriteRendererProperties
    &NiagaraMeshRendererProperties
    &NiagaraLightRendererProperties
   ]
  }
  {emitterName=Embers, rendererClasses [&NiagaraSpriteRendererProperties]}
  {emitterName=Thruster_Spikes, rendererClasses [&NiagaraMeshRendererProperties]}
  {emitterName=Smoke_Puff, rendererClasses [&NiagaraMeshRendererProperties]}
  {emitterName=Motion_Trails, rendererClasses [&NiagaraRibbonRendererProperties]}
 ]
 userVariables [
  User.bIsFrontEnd {type=NiagaraBool}
  User.bIsFrontEndPreview {type=NiagaraBool}
  User.ForwardDot {type=NiagaraFloat, defaultValue=0, description=""}
  User.RibbonIDOffset {type=NiagaraInt32}
  User.RightDot {type=NiagaraFloat, defaultValue=0, description=""}
 ]
}
