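# Control Rig asset for the BoomShot glider, rigging its base, frame, body offset, three grenade bones and two rings.
# The connector names indicate a modular rig: a Root module plus per-part modules, each with primary and parent control connectors.
# Each bone has a socket, and the left and right grenade branches hang off the center grenade, each carrying a ring.
# The rig's own graph holds only an unconnected forward-solve entry, with no logic of its own.
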ControlRigBlueprint /BRCosmetics/Gadgets/Assets/VinderTech_GliderChute/Glider_BoomShot/Meshes/MR_BoomShot_Glider_CtrlRig
graphs [
 RigVMFunctionLibrary {links [], nodes []}
 RigVMModel {nodes [RigUnit_BeginExecution = &RigUnit_BeginExecution()]}
]
hierarchy [
 Root {loc=<0 0 0>, rot=<0 0 0 1>}
 Base {parent=Root, loc=<0.0034 -0.0172 10.8432>, rot=<0 0 0 1>}
 Frame {parent=Base, loc=<-0.0034 11.5951 -3.7768>, rot=<0 0 0 1>}
 Body_Offset {parent=Frame, loc=<0 0 0>, rot=<0 0 0 1>}
 Grenade_c_01 {parent=Body_Offset, loc=<0 -13.6337 -58.0131>, rot=<0 0 0 1>}
 Grenade_l_01 {parent=Grenade_c_01, loc=<47.3437 0 0>, rot=<0 0 0 1>}
 RingB {parent=Grenade_l_01, loc=<0 -94.1708 -18.7236>, rot=<0 0 0 1>}
 RingB_socket {parent=RingB, type=Socket, loc=<0 0 0>, rot=<0 0 0 1>}
 Grenade_l_01_socket {parent=Grenade_l_01, type=Socket, loc=<0 0 0>, rot=<0 0 0 1>}
 Grenade_r_01 {parent=Grenade_c_01, loc=<-47.3437 0 0>, rot=<1 0 0 0>}
 RingA {parent=Grenade_r_01, loc=<-0.0832 94.1708 18.7236>, rot=<1 0 0 0>}
 RingA_socket {parent=RingA, type=Socket, loc=<0 0 0>, rot=<0 0 0 1>}
 Grenade_r_01_socket {parent=Grenade_r_01, type=Socket, loc=<0 0 0>, rot=<0 0 0 1>}
 Grenade_c_01_socket {parent=Grenade_c_01, type=Socket, loc=<0 0 0>, rot=<0 0 0 1>}
 Body_Offset_socket {parent=Body_Offset, type=Socket, loc=<0 0 0>, rot=<0 0 0 1>}
 Frame_socket {parent=Frame, type=Socket, loc=<0 0 0>, rot=<0 0 0 1>}
 Base_socket {parent=Base, type=Socket, loc=<0 0 0>, rot=<0 0 0 1>}
 Root_socket {parent=Root, type=Socket, loc=<0 0 0>, rot=<0 0 0 1>}
 Root/Root {type=Connector, loc=<0 0 0>, rot=<0 0 0 1>}
 Glider_Base/Add Control Primary {type=Connector, loc=<0 0 0>, rot=<0 0 0 1>}
 Glider_Base/Parent Control {type=Connector, loc=<0 0 0>, rot=<0 0 0 1>}
 Frame/Add Control Primary {type=Connector, loc=<0 0 0>, rot=<0 0 0 1>}
 Frame/Parent Control {type=Connector, loc=<0 0 0>, rot=<0 0 0 1>}
 Body_Offset/Add Control Primary {type=Connector, loc=<0 0 0>, rot=<0 0 0 1>}
 Body_Offset/Parent Control {type=Connector, loc=<0 0 0>, rot=<0 0 0 1>}
 GrenadeC/Add Control Primary {type=Connector, loc=<0 0 0>, rot=<0 0 0 1>}
 GrenadeC/Parent Control {type=Connector, loc=<0 0 0>, rot=<0 0 0 1>}
 GrenadeR/Add Control Primary {type=Connector, loc=<0 0 0>, rot=<0 0 0 1>}
 GrenadeR/Parent Control {type=Connector, loc=<0 0 0>, rot=<0 0 0 1>}
 GrenadeL/Add Control Primary {type=Connector, loc=<0 0 0>, rot=<0 0 0 1>}
 GrenadeL/Parent Control {type=Connector, loc=<0 0 0>, rot=<0 0 0 1>}
 RingA/Add Control Primary {type=Connector, loc=<0 0 0>, rot=<0 0 0 1>}
 RingA/Parent Control {type=Connector, loc=<0 0 0>, rot=<0 0 0 1>}
 RingB/Add Control Primary {type=Connector, loc=<0 0 0>, rot=<0 0 0 1>}
 RingB/Parent Control {type=Connector, loc=<0 0 0>, rot=<0 0 0 1>}
]
